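# Parent cloth-shading material for the StairGummy emote prop, built on the shared 2023 anime shader functions.
# Pre, Calculations and Post functions carry the shading, with masked eye animation on color and an IsKicks switch for sock masking.
# A LightFlash node chain draws a polar ring of alternately flashing dots, gated by ToggleFlash, into the FlashLightDots reroute.
# That reroute and two Custom HLSL versions of the effect have no visible consumers.
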Material /BRCosmetics/Animation/Props/StairGummy/Materials/M_EmoteProp_StairGummy_PARENT
graph {
 expressions {
  MaterialExpressionAbs_0 {class=MaterialExpressionAbs, inputs [None {from=MaterialExpressionNamedRerouteUsage_0, output=""}], properties {}}
  MaterialExpressionAdd_4 {
   class=MaterialExpressionAdd
   inputs [
    A {from=MaterialExpressionDivide_1}
    B {from=""}
   ]
   properties {constB=0.5}
  }
  MaterialExpressionAdd_5 {
   class=MaterialExpressionAdd
   inputs [
    A {from=MaterialExpressionScalarParameter_5}
    B {from=MaterialExpressionScalarParameter_11}
   ]
   properties {constB=0.005}
  }
  MaterialExpressionAdd_6 {
   class=MaterialExpressionAdd
   inputs [
    A {from=MaterialExpressionMultiply_3}
    B {from=""}
   ]
   properties {constB=0.5, desc=baseSignal}
  }
  MaterialExpressionAppendVector_0 {
   class=MaterialExpressionAppendVector
   inputs [
    A {from=MaterialExpressionVertexColor_0, output=""}
    B {from=MaterialExpressionVertexColor_0, output=""}
   ]
  }
  MaterialExpressionArctangent2_0 {
   class=MaterialExpressionArctangent2
   inputs [
    Y {from=MaterialExpressionConvert_1, output=""}
    X {from=MaterialExpressionConvert_1, output=""}
   ]
   properties {desc=angle}
  }
  MaterialExpressionArctangent_0 {class=MaterialExpressionArctangent, inputs [None {from="", output=""}], properties {}}
  MaterialExpressionChannelMaskParameter_0 {
   class=MaterialExpressionChannelMaskParameter
   inputs [None {from=MaterialExpressionStaticSwitchParameter_1, output=""}]
   properties {group=01_EyeAnimation_Parameters, parameterName=EyeMaskChannel, sortPriority=-32}
  }
  MaterialExpressionComponentMask_0 {class=MaterialExpressionComponentMask, inputs [None {from=MaterialExpressionTextureCoordinate_0}], properties {r=true}}
  MaterialExpressionConstant_3 {class=MaterialExpressionConstant, properties {desc=tau, r=6.2832}}
  MaterialExpressionConstant_4 {class=MaterialExpressionConstant, properties {r=22}}
  MaterialExpressionConstant_5 {class=MaterialExpressionConstant, properties {r=0.123}}
  MaterialExpressionConstant_6 {class=MaterialExpressionConstant, properties {r=0.007}}
  MaterialExpressionConstant_7 {class=MaterialExpressionConstant, properties {r=20}}
  MaterialExpressionConstant_9 {class=MaterialExpressionConstant, properties {r=2}}
  MaterialExpressionConvert_0 {
   class=MaterialExpressionConvert
   inputs [
    None {from=MaterialExpressionVectorParameter_0, output=R}
    None {from=MaterialExpressionVectorParameter_0, output=R}
   ]
   properties {
    convertInputs [
     {
      type=Scalar
      expressionInput {
       expression=&self:MaterialExpressionVectorParameter_0
       inputName=None
       mask=1
       maskA=0
       maskB=0
       maskG=0
       maskR=1
       outputIndex=1
      }
     }
     {
      type=Scalar
      expressionInput {
       expression=&self:MaterialExpressionVectorParameter_0
       inputName=None
       mask=1
       maskA=0
       maskB=0
       maskG=1
       maskR=0
       outputIndex=2
      }
     }
    ]
    convertOutputs [{type=Vector2}]
   }
  }
  MaterialExpressionConvert_1 {
   class=MaterialExpressionConvert
   inputs [None {from=MaterialExpressionSubtract_2}]
   properties {
    convertInputs [{
      type=Vector2
      expressionInput {
       expression=&self:MaterialExpressionSubtract_2
       inputName=None
       mask=0
       maskA=0
       maskB=0
       maskG=0
       maskR=0
       outputIndex=0
      }
     }]
    convertOutputs [{type=Scalar}, {type=Scalar}]
   }
  }
  MaterialExpressionConvert_2 {
   class=MaterialExpressionConvert
   inputs [
    None {from=MaterialExpressionVectorParameter_1, output=R}
    None {from=MaterialExpressionVectorParameter_1, output=R}
   ]
   properties {
    convertInputs [
     {
      type=Scalar
      expressionInput {
       expression=&self:MaterialExpressionVectorParameter_1
       inputName=None
       mask=1
       maskA=0
       maskB=0
       maskG=0
       maskR=1
       outputIndex=1
      }
     }
     {
      type=Scalar
      expressionInput {
       expression=&self:MaterialExpressionVectorParameter_1
       inputName=None
       mask=1
       maskA=0
       maskB=0
       maskG=1
       maskR=0
       outputIndex=2
      }
     }
    ]
    convertOutputs [{type=Vector2}]
   }
  }
  MaterialExpressionConvert_3 {
   class=MaterialExpressionConvert
   inputs [
    None {from=MaterialExpressionCosine_0}
    None {from=MaterialExpressionSine_0}
   ]
   properties {
    nodeName=Make Vector2
    convertInputs [
     {
      type=Scalar
      expressionInput {
       expression=&self:MaterialExpressionCosine_0
       inputName=None
       mask=0
       maskA=0
       maskB=0
       maskG=0
       maskR=0
       outputIndex=0
      }
     }
     {
      type=Scalar
      expressionInput {
       expression=&self:MaterialExpressionSine_0
       inputName=None
       mask=0
       maskA=0
       maskB=0
       maskG=0
       maskR=0
       outputIndex=0
      }
     }
    ]
    convertOutputs [{type=Vector2}]
   }
  }
  MaterialExpressionCosine_0 {class=MaterialExpressionCosine, inputs [None {from=MaterialExpressionMultiply_0, output=""}], properties {period=6.2832}}
  MaterialExpressionCustom_1 {
   class=MaterialExpressionCustom
   inputs [
    uv {from="", output=""}
    count {from="", output=""}
    dotSize {from="", output=""}
    ringRadius {from="", output=""}
   ]
   properties {
    code="// The below expression will get compiled\r\n// into the output of this node    \r\nfloat blur = 0.001;         \r\n// ----------------\r\n\r\n// 2. 极坐标转换\r\n// atan(y, x) 返回范围是 -PI 到 PI\r\nfloat angle = atan2(uv.y, uv.x);\r\nfloat tau = 6.28318530718; // 2 * PI\r\n    \r\n// 3. 角度离散化 (计算当前像素属于哪一个圆点的“势力范围”)\r\n// 将角度从 [-PI, PI] 映射到 [0, 1]\r\nfloat slice = tau / count;\r\nfloat segment = floor(angle / slice + 0.5); \r\n    \r\n// 4. 计算该圆点在笛卡尔坐标系中的位置\r\nfloat targetAngle = segment * slice;\r\nfloat2 dotCenter = float2(cos(targetAngle), sin(targetAngle)) * ringRadius;\r\n\r\n// 5. 画圆 (SDF - Signed Distance Function)\r\nfloat dist = length(uv - dotCenter);\r\n    \r\nfloat circle = smoothstep(dotSize + blur, dotSize - blur, dist);\r\n\r\n// 输出颜色\r\n//float3 color = float3(circle) * float3(0.0, 0.7, 1.0); // 蓝色小球\r\nreturn circle;"
    showCode=true
    additionalOutputs [{outputName=color, outputType=CMOT_Float1}]
    inputs [
     {inputName=uv, input {expression=null, inputName=None, mask=0, maskA=0, maskB=0, maskG=0, maskR=0, outputIndex=0}}
     {inputName=count, input {expression=null, inputName=None, mask=0, maskA=0, maskB=0, maskG=0, maskR=0, outputIndex=0}}
     {inputName=dotSize, input {expression=null, inputName=None, mask=0, maskA=0, maskB=0, maskG=0, maskR=0, outputIndex=0}}
     {inputName=ringRadius, input {expression=null, inputName=None, mask=0, maskA=0, maskB=0, maskG=0, maskR=0, outputIndex=0}}
    ]
   }
  }
  MaterialExpressionCustom_2 {
   class=MaterialExpressionCustom
   inputs [
    UV {from=MaterialExpressionSubtract_1, output=""}
    Time {from=MaterialExpressionTime_0, output=""}
    Count {from=MaterialExpressionConstant_4, output=""}
    RingRadius {from=MaterialExpressionConstant_5, output=""}
    DotSize {from=MaterialExpressionConstant_6, output=""}
    FlashSpeed {from=MaterialExpressionConstant_7, output=""}
   ]
   properties {
    code="// 1. Polar coordinate conversion\r\nfloat tau = 6.28318530718; \r\nfloat angle = atan2(UV.y, UV.x); \r\n\r\n// 2. Angular discretization (Index calculation)\r\nfloat slice = tau / Count;\r\n// Use floor + 0.5 for rounding to locate the nearest dot inde\r\nfloat index = floor(angle / slice + 0.5);\r\n\r\n// 4. Calculate the center position coordinates for the current dot index\r\nfloat targetAngle = index * slice;\r\nfloat2 center = float2(cos(targetAngle), sin(targetAngle)) * RingRadius;\r\n\r\n// 5. Draw the circle (SDF)\r\nfloat dist = length(UV - center);\r\n// Antialiasing / Smooth edge processing\r\nfloat circle = smoothstep(DotSize + 0.005, DotSize - 0.005, dist);\r\n\r\n// 6. Regular alternating flash logic\r\n// Calculate the base waveform (ranging from 0.0 to 1.0)\r\nfloat baseSignal = sin(Time * FlashSpeed) * 0.5 + 0.5;\r\n\r\n// Check parity: using fmod. abs() handles potential negative index cases\r\nfloat isEven = fmod(abs(index), 2.0) < 1.0 ? 1.0 : 0.0;\r\n\r\n// Implement alternation: even indices follow the wave, odd indices are inverted\r\nfloat brightness = (isEven > 0.5) ? baseSignal : (1.0 - baseSignal);\r\n\r\n// Return the final result (can be used as Alpha or Emissive Intensity)\r\nreturn circle * brightness;"
    showCode=true
    additionalOutputs [{outputName=None, outputType=CMOT_Float1}]
    inputs [
     {
      inputName=UV
      input {
       expression=&self:MaterialExpressionSubtract_1
       inputName=None
       mask=0
       maskA=0
       maskB=0
       maskG=0
       maskR=0
       outputIndex=0
      }
     }
     {
      inputName=Time
      input {
       expression=&self:MaterialExpressionTime_0
       inputName=None
       mask=0
       maskA=0
       maskB=0
       maskG=0
       maskR=0
       outputIndex=0
      }
     }
     {
      inputName=Count
      input {
       expression=&self:MaterialExpressionConstant_4
       inputName=None
       mask=0
       maskA=0
       maskB=0
       maskG=0
       maskR=0
       outputIndex=0
      }
     }
     {
      inputName=RingRadius
      input {
       expression=&self:MaterialExpressionConstant_5
       inputName=None
       mask=0
       maskA=0
       maskB=0
       maskG=0
       maskR=0
       outputIndex=0
      }
     }
     {
      inputName=DotSize
      input {
       expression=&self:MaterialExpressionConstant_6
       inputName=None
       mask=0
       maskA=0
       maskB=0
       maskG=0
       maskR=0
       outputIndex=0
      }
     }
     {
      inputName=FlashSpeed
      input {
       expression=&self:MaterialExpressionConstant_7
       inputName=None
       mask=0
       maskA=0
       maskB=0
       maskG=0
       maskR=0
       outputIndex=0
      }
     }
    ]
   }
  }
  MaterialExpressionDivide_0 {
   class=MaterialExpressionDivide
   inputs [
    A {from=MaterialExpressionConstant_3}
    B {from=MaterialExpressionScalarParameter_3}
   ]
   properties {desc=Slice}
  }
  MaterialExpressionDivide_1 {
   class=MaterialExpressionDivide
   inputs [
    A {from=MaterialExpressionArctangent2_0}
    B {from=MaterialExpressionDivide_0}
   ]
  }
  MaterialExpressionFloor_0 {class=MaterialExpressionFloor, inputs [None {from=MaterialExpressionAdd_4, output=""}], properties {desc=Index/Segment}}
  MaterialExpressionFmod_0 {
   class=MaterialExpressionFmod
   inputs [
    A {from=MaterialExpressionAbs_0, output=""}
    B {from=MaterialExpressionConstant_9, output=""}
   ]
   properties {}
  }
  MaterialExpressionLength_0 {class=MaterialExpressionLength, inputs [None {from=MaterialExpressionSubtract_3, output=""}], properties {}}
  MaterialExpressionLinearInterpolate_2 {
   class=MaterialExpressionLinearInterpolate
   inputs [
    A {from=""}
    B {from=""}
    Alpha {from=MaterialExpressionStep_1}
   ]
   properties {desc=isEven}
  }
  MaterialExpressionLinearInterpolate_3 {
   class=MaterialExpressionLinearInterpolate
   inputs [
    A {from=MaterialExpressionAdd_6}
    B {from=MaterialExpressionOneMinus_0}
    Alpha {from=MaterialExpressionStep_2}
   ]
   properties {desc=Brightness}
  }
  MaterialExpressionMaterialFunctionCall_0 {
   class=MaterialExpressionMaterialFunctionCall
   inputs [
    Color Lit {from=MaterialExpressionMaterialFunctionCall_21, output=Color_Lit}
    Color Shaded {from=MaterialExpressionMaterialFunctionCall_21, output=Color_Lit}
    Use UV2 {from=MaterialExpressionStaticBoolParameter_0}
    Mask {from=MaterialExpressionStaticSwitchParameter_0}
   ]
   properties {materialFunction=/Game/Global/Materials/Character_Parent_2023/Anime_Parent_2023/MaterialFunctions/MF_AnimeEyeAnimation_v2}
  }
  MaterialExpressionMaterialFunctionCall_14 {
   class=MaterialExpressionMaterialFunctionCall
   inputs [
    Shading_Gradient {from=MaterialExpressionMaterialFunctionCall_20, output=Shading_Gradient}
    Threshold_Gradient {from=MaterialExpressionMaterialFunctionCall_20, output=Shading_Gradient}
    Specular_Gradient {from=MaterialExpressionMaterialFunctionCall_20, output=Shading_Gradient}
    Sheen_Mask {from=MaterialExpressionMaterialFunctionCall_20, output=Shading_Gradient}
    Thickness {from=MaterialExpressionMaterialFunctionCall_20, output=Shading_Gradient}
    DFL_Mask {from=MaterialExpressionMaterialFunctionCall_20, output=Shading_Gradient}
    FX {from=MaterialExpressionMaterialFunctionCall_20, output=Shading_Gradient}
    Metallic_Gradient {from=MaterialExpressionMaterialFunctionCall_20, output=Shading_Gradient}
    RimLight_Gradient {from=MaterialExpressionMaterialFunctionCall_20, output=Shading_Gradient}
    NormalBuffer {from=MaterialExpressionMaterialFunctionCall_20, output=Shading_Gradient}
    -------- {from=""}
    CINE_LightSources {from=MaterialExpressionMaterialFunctionCall_20, output=Shading_Gradient}
    CINE_LightColor {from=MaterialExpressionMaterialFunctionCall_20, output=Shading_Gradient}
    --------- {from=""}
    Color_Lit {from=MaterialExpressionMaterialFunctionCall_0, output=Lit}
    Color_Shaded {from=MaterialExpressionMaterialFunctionCall_0, output=Lit}
    SRM {from=MaterialExpressionMaterialFunctionCall_21, output=SRM}
    Emissive {from=MaterialExpressionMaterialFunctionCall_21, output=SRM}
    DFL_Color {from=MaterialExpressionMaterialFunctionCall_21, output=SRM}
    Specular_Color {from=MaterialExpressionMaterialFunctionCall_21, output=SRM}
    RimLight_Color {from=MaterialExpressionMaterialFunctionCall_21, output=SRM}
   ]
   properties {materialFunction=/Game/Global/Materials/Character_Parent_2023/Anime_Parent_2023/MaterialFunctions/MFC_AnimeShader_2023_Post}
  }
  MaterialExpressionMaterialFunctionCall_2 {
   class=MaterialExpressionMaterialFunctionCall
   inputs [
    InOpacity {from=""}
    Material {from=MaterialExpressionMaterialFunctionCall_14, output=Result}
    Mask {from=""}
   ]
   properties {materialFunction=/Game/Global/Materials/Cosmetics/Kicks/Socks/Materials/MF_Sock}
  }
  MaterialExpressionMaterialFunctionCall_20 {
   class=MaterialExpressionMaterialFunctionCall
   inputs [
    CustomLightVector {from=""}
    OverrideNormal_WS {from=""}
    STT {from=MaterialExpressionMaterialFunctionCall_21, output=STT}
    Normal_Map {from=MaterialExpressionMaterialFunctionCall_21, output=STT}
    DFL {from=MaterialExpressionMaterialFunctionCall_21, output=STT}
    FX {from=MaterialExpressionMaterialFunctionCall_21, output=STT}
   ]
   properties {materialFunction=/Game/Global/Materials/Character_Parent_2023/Anime_Parent_2023/MaterialFunctions/MFC_AnimeShader_2023_Calculations}
  }
  MaterialExpressionMaterialFunctionCall_21 {
   class=MaterialExpressionMaterialFunctionCall
   inputs [CustomUV {from=""}]
   properties {materialFunction=/Game/Global/Materials/Character_Parent_2023/Anime_Parent_2023/MaterialFunctions/MFC_AnimeShader_2023_Pre}
  }
  MaterialExpressionMultiply_0 {
   class=MaterialExpressionMultiply
   inputs [
    A {from=MaterialExpressionNamedRerouteDeclaration_0}
    B {from=MaterialExpressionDivide_0}
   ]
   properties {desc=targetAngle}
  }
  MaterialExpressionMultiply_1 {
   class=MaterialExpressionMultiply
   inputs [
    A {from=MaterialExpressionConvert_3}
    B {from=MaterialExpressionScalarParameter_4}
   ]
   properties {desc=dot Center}
  }
  MaterialExpressionMultiply_2 {
   class=MaterialExpressionMultiply
   inputs [
    A {from=MaterialExpressionTime_1}
    B {from=MaterialExpressionScalarParameter_7}
   ]
  }
  MaterialExpressionMultiply_3 {
   class=MaterialExpressionMultiply
   inputs [
    A {from=MaterialExpressionSine_1}
    B {from=""}
   ]
   properties {constB=0.5}
  }
  MaterialExpressionMultiply_4 {
   class=MaterialExpressionMultiply
   inputs [
    A {from=MaterialExpressionPower_0}
    B {from=MaterialExpressionMultiply_7}
   ]
  }
  MaterialExpressionMultiply_5 {
   class=MaterialExpressionMultiply
   inputs [
    A {from=MaterialExpressionMultiply_4}
    B {from=MaterialExpressionScalarParameter_8}
   ]
  }
  MaterialExpressionMultiply_6 {
   class=MaterialExpressionMultiply
   inputs [
    A {from=MaterialExpressionLinearInterpolate_3}
    B {from=MaterialExpressionVectorParameter_2, output=RGB}
   ]
  }
  MaterialExpressionMultiply_7 {
   class=MaterialExpressionMultiply
   inputs [
    A {from=MaterialExpressionMultiply_6}
    B {from=MaterialExpressionScalarParameter_10}
   ]
  }
  MaterialExpressionNamedRerouteDeclaration_0 {
   class=MaterialExpressionNamedRerouteDeclaration
   inputs [None {from=MaterialExpressionFloor_0}]
   properties {name=Index, nodeColor=<rgba 1 0 0.8906 1>}
  }
  MaterialExpressionNamedRerouteDeclaration_1 {
   class=MaterialExpressionNamedRerouteDeclaration
   inputs [None {from=MaterialExpressionMultiply_5}]
   properties {name=FlashLightDots, nodeColor=<rgba 1 0 0.8672 1>}
  }
  MaterialExpressionNamedRerouteUsage_0 {class=MaterialExpressionNamedRerouteUsage}
  MaterialExpressionOneMinus_0 {class=MaterialExpressionOneMinus, inputs [None {from=MaterialExpressionAdd_6}]}
  MaterialExpressionPower_0 {
   class=MaterialExpressionPower
   inputs [
    Base {from=MaterialExpressionSmoothStep_0}
    Exp {from=MaterialExpressionScalarParameter_12}
   ]
  }
  MaterialExpressionSaturate_0 {class=MaterialExpressionSaturate, inputs [None {from=MaterialExpressionSign_0}]}
  MaterialExpressionScalarParameter_10 {class=MaterialExpressionScalarParameter, properties {defaultValue=1, group=LightFlash, parameterName=LightsIntensity}}
  MaterialExpressionScalarParameter_11 {class=MaterialExpressionScalarParameter, properties {defaultValue=0.005, group=LightFlash, parameterName=DotSoftness}}
  MaterialExpressionScalarParameter_12 {class=MaterialExpressionScalarParameter, properties {defaultValue=2, group=LightFlash, parameterName=DotExp}}
  MaterialExpressionScalarParameter_3 {class=MaterialExpressionScalarParameter, properties {defaultValue=22, group=LightFlash, parameterName=DotCount}}
  MaterialExpressionScalarParameter_4 {class=MaterialExpressionScalarParameter, properties {defaultValue=0.123, group=LightFlash, parameterName=RingRadius}}
  MaterialExpressionScalarParameter_5 {class=MaterialExpressionScalarParameter, properties {defaultValue=0.007, group=LightFlash, parameterName=DotSize}}
  MaterialExpressionScalarParameter_7 {class=MaterialExpressionScalarParameter, properties {defaultValue=20, group=LightFlash, parameterName=FlashSpeed}}
  MaterialExpressionScalarParameter_8 {class=MaterialExpressionScalarParameter, properties {group=LightFlash, parameterName=ToggleFlash}}
  MaterialExpressionSign_0 {class=MaterialExpressionSign, inputs [None {from=MaterialExpressionSubtract_0, output=""}], properties {}}
  MaterialExpressionSine_0 {class=MaterialExpressionSine, inputs [None {from=MaterialExpressionMultiply_0}], properties {period=6.2832}}
  MaterialExpressionSine_1 {class=MaterialExpressionSine, inputs [None {from=MaterialExpressionMultiply_2}], properties {period=6.2832}}
  MaterialExpressionSmoothStep_0 {
   class=MaterialExpressionSmoothStep
   inputs [
    Min {from=MaterialExpressionAdd_5, output=""}
    Max {from=MaterialExpressionSubtract_4, output=""}
    Value {from=MaterialExpressionLength_0, output=""}
   ]
   properties {}
  }
  MaterialExpressionStaticBoolParameter_0 {
   class=MaterialExpressionStaticBoolParameter
   inputs []
   properties {defaultValue=true, group=01_EyeAnimation_Parameters, parameterName=Use UV2 for Eye Animation, sortPriority=-30}
  }
  MaterialExpressionStaticSwitchParameter_0 {
   class=MaterialExpressionStaticSwitchParameter
   inputs [
    True {from=MaterialExpressionChannelMaskParameter_0, output=R}
    False {from=MaterialExpressionSaturate_0, output=""}
   ]
   properties {group=01_EyeAnimation_Parameters, parameterName=UseEyeMask, sortPriority=-33}
  }
  MaterialExpressionStaticSwitchParameter_1 {
   class=MaterialExpressionStaticSwitchParameter
   inputs [
    True {from=MaterialExpressionAppendVector_0, output=""}
    False {from=MaterialExpressionTextureSampleParameter2D_0, output=RGBA}
   ]
   properties {group=01_EyeAnimation_Parameters, parameterName=UseVertexColorForEyeMask, sortPriority=-33}
  }
  MaterialExpressionStaticSwitchParameter_2 {
   class=MaterialExpressionStaticSwitchParameter
   inputs [
    True {from=MaterialExpressionMaterialFunctionCall_2, output=Result}
    False {from=MaterialExpressionMaterialFunctionCall_14, output=Result}
   ]
   properties {group=Kicks, parameterName=IsKicks}
  }
  MaterialExpressionStep_0 {
   class=MaterialExpressionStep
   inputs [
    Y {from=MaterialExpressionSubtract_3, output=""}
    X {from="", output=""}
   ]
   properties {constX=0.01}
  }
  MaterialExpressionStep_1 {
   class=MaterialExpressionStep
   inputs [
    Y {from="", output=""}
    X {from=MaterialExpressionFmod_0, output=""}
   ]
   properties {constY=1}
  }
  MaterialExpressionStep_2 {
   class=MaterialExpressionStep
   inputs [
    Y {from=MaterialExpressionLinearInterpolate_2, output=""}
    X {from="", output=""}
   ]
   properties {constX=0.5}
  }
  MaterialExpressionSubtract_0 {
   class=MaterialExpressionSubtract
   inputs [
    A {from=MaterialExpressionComponentMask_0}
    B {from=""}
   ]
   properties {constB=0.0001}
  }
  MaterialExpressionSubtract_1 {
   class=MaterialExpressionSubtract
   inputs [
    A {from=MaterialExpressionTextureCoordinate_1}
    B {from=MaterialExpressionConvert_0}
   ]
  }
  MaterialExpressionSubtract_2 {
   class=MaterialExpressionSubtract
   inputs [
    A {from=MaterialExpressionTextureCoordinate_3}
    B {from=MaterialExpressionConvert_2}
   ]
  }
  MaterialExpressionSubtract_3 {
   class=MaterialExpressionSubtract
   inputs [
    A {from=MaterialExpressionSubtract_2}
    B {from=MaterialExpressionMultiply_1}
   ]
  }
  MaterialExpressionSubtract_4 {
   class=MaterialExpressionSubtract
   inputs [
    A {from=MaterialExpressionScalarParameter_5}
    B {from=MaterialExpressionScalarParameter_11}
   ]
   properties {constB=0.005}
  }
  MaterialExpressionTemporalResponsivenessOutput_0 {class=MaterialExpressionTemporalResponsivenessOutput, inputs [None {from="", output=""}], properties {}}
  MaterialExpressionTextureCoordinate_0 {class=MaterialExpressionTextureCoordinate, properties {coordinateIndex=1}}
  MaterialExpressionTextureCoordinate_1 {class=MaterialExpressionTextureCoordinate}
  MaterialExpressionTextureCoordinate_3 {class=MaterialExpressionTextureCoordinate}
  MaterialExpressionTextureSampleParameter2D_0 {
   class=MaterialExpressionTextureSampleParameter2D
   inputs [
    UVs {from="", output=""}
    Apply View MipBias {from="", output=""}
   ]
   properties {
    group=01_EyeAnimation_Parameters
    parameterName=EyeMask
    samplerSource=SSM_Wrap_WorldGroupSettings
    samplerType=SAMPLERTYPE_Masks
    sortPriority=-32
    texture=&T_WhiteMask
   }
  }
  MaterialExpressionTime_0 {class=MaterialExpressionTime}
  MaterialExpressionTime_1 {class=MaterialExpressionTime}
  MaterialExpressionVectorParameter_0 {class=MaterialExpressionVectorParameter, properties {group=LightFlash, parameterName=RingCenterOffset, defaultValue=<rgba 0.1813 0.672 0 1>}}
  MaterialExpressionVectorParameter_1 {class=MaterialExpressionVectorParameter, properties {group=LightFlash, parameterName=RingCenterOffset, defaultValue=<rgba 0.1813 0.672 0 1>}}
  MaterialExpressionVectorParameter_2 {class=MaterialExpressionVectorParameter, properties {group=LightFlash, parameterName=LightsColor, defaultValue=<rgba 1 1 1 1>}}
  MaterialExpressionVertexColor_0 {class=MaterialExpressionVertexColor, inputs [], properties {}}
 }
 properties {
  bAutomaticallySetUsageInEditor=false
  bUseMaterialAttributes=true
  bUsedWithClothing=true
  bUsedWithMeshDeformer=true
  bUsedWithMorphTargets=true
  bUsedWithNiagaraMeshParticles=true
  bUsedWithSkeletalMesh=true
  bUsedWithStaticLighting=true
  pixelDepthOffsetMode=PDOM_Legacy
  shadingModel=MSM_Cloth
  usedShadingModels=Cloth
 }
}
